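# BasilOkra_Triangle_Mat is a Default Lit material in a Temp folder of the WildEstate_BasilOkra content; its in-game use is not shown.
# Its graph holds a single texture sample of the BasilOkra_Triangle texture, with no connection to a material output shown.
# Automatic setting of usage flags in the editor is turned off.
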Material /WildEstate_BasilOkra/Assets/Temp/Materials/BasilOkra_Triangle_Mat
graph {
 properties {bAutomaticallySetUsageInEditor=false, usedShadingModels=Default Lit}
 nodes [TextureSample() {texture=/WildEstate_BasilOkra/Assets/Temp/Textures/BasilOkra_Triangle}]
}
