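# Unlit, translucent, two-sided particle material for an alarm cone effect, usable on sprites and mesh particles.
# Particle color RGB passes through MF_FX_EmissiveAdjustments, while particle alpha is multiplied by a soft vertical ramp texture and a view-angle fade with adjustable sensitivity.
# A feature-level switch adds depth fade to that opacity on default and skips it on ES2 and ES3.1.
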Material /WildEstate_BasilOkra/Assets/Temp/Materials/M_Alarm_Cone_BasilOkra
graph {
 properties {
  bUsedWithMeshParticles=true
  bUsedWithNiagaraMeshParticles=true
  bUsedWithParticleSprites=true
  blendMode=BLEND_Translucent
  pixelDepthOffsetMode=PDOM_Legacy
  shadingModel=MSM_Unlit
  twoSided=true
  usedShadingModels=Unlit
 }
 nodes [
  n1 = ParticleColor()
  n2 = Multiply(
   A=Multiply(A=TextureSample() {texture=/Game/Effects/Fort_Effects/Textures/Ramps/T_Ramp_Vert_Soft_Fade}.G, B=@n1.A)
   B=MaterialFunctionCall("FadeSensitivity (optional)"=ScalarParameter() {defaultValue=4, parameterName=FadeSensitivity}) {materialFunction=&FadeBasedOnViewAngle}.Result
  )
  n3 = DepthFade(Opacity=@n2, FadeDistance=ScalarParameter() {defaultValue=30, parameterName=FadeDistance})
  FeatureLevelSwitch(Default=@n3, ES2=@n2, ES3_1=@n2, SM4_REMOVED=@n3)
  MaterialFunctionCall(Emissive=@n1.RGB) {materialFunction=&MF_FX_EmissiveAdjustments}
  Multiply()
 ]
}
