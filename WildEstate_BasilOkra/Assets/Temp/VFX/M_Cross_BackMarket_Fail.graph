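# Unlit, translucent, two-sided material for Niagara mesh particles, apparently showing a cross for a failed black market purchase.
# The red, green and blue channels of the CrossTex texture are each multiplied by particle color.
# Lerp_3Color blends these three variants by the WhichCross? scalar, letting one texture hold three crosses.
# The blend feeds the opacity input of a DepthFade node, which softens intersections with scene geometry.
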Material /WildEstate_BasilOkra/Assets/Temp/VFX/M_Cross_BackMarket_Fail
graph {
 properties {
  bAutomaticallySetUsageInEditor=false
  bUsedWithNiagaraMeshParticles=true
  blendMode=BLEND_Translucent
  shadingModel=MSM_Unlit
  twoSided=true
  usedShadingModels=Unlit
 }
 nodes [
  n1 = ParticleColor()
  n2 = TextureSampleParameter2D() {parameterName=CrossTex, texture=&/T_Cross_FailPurchase}
  DepthFade(Opacity=MaterialFunctionCall(
    A=Multiply(A=@n1.RGB, B=@n2.R)
    B=Multiply(A=@n1.RGB, B=@n2.G)
    C=Multiply(A=@n1.RGB, B=@n2.B)
    Alpha=ScalarParameter() {parameterName=WhichCross?}
   ) {materialFunction=&Lerp_3Color}.Result) {fadeDistanceDefault=25}
 ]
}
